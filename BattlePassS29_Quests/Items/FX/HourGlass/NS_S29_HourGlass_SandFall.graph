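# Niagara system for sand falling inside an hourglass, part of the Season 29 Battle Pass quest items.
# Sand_Fall spawns SubUV sand sprites that drop under gravity, while Mist adds rotating dust puffs.
# User.HourGlassSize lerps spawn rates, sand start height, gravity, scale and camera offset, with mist spawning dropping to zero at 1.
# Both emitters tint by time of day and feed dynamic material parameters.
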NiagaraSystem /BattlePassS29_Quests/Items/FX/HourGlass/NS_S29_HourGlass_SandFall
dependencies {
 usedDataInterfaces [
  &NiagaraDataInterfaceCurve
  &NiagaraDataInterfaceColorCurve
  &NiagaraDataInterfaceSpriteRendererInfo
  &NiagaraDataInterfaceMeshRendererInfo
  &NiagaraDataInterfaceVector2DCurve
 ]
 usedDynamicInputs [&Vector2DFromFloat]
 usedModules [&SystemState]
}
emitters {
 Mist {
  emitterUpdateScript [
   &EmitterState()
   &SpawnRate(SpawnRate=&Lerp_Float(A=12, B=0, Alpha=$User.HourGlassSize), Use Spawn Probability=false)
  ]
  particleSpawnScript [
   &InitializeParticle(
    Color=<rgba 0.6514 0.3663 0.0999 1>
    Lifetime Max=2.5
    Lifetime Min=2
    Lifetime Mode=::NewEnumerator1
    Position Offset=<0 0 80>
    Sprite Rotation Mode=::NewEnumerator2
    Sprite Size Mode=::NewEnumerator3
    Uniform Sprite Size=20
    UsePositionOffset=true
    Write Material Random=true
   )
   &DynamicMaterialParameters(
    Index 0 Param 1=&UniformRangedFloat(Minimum=1, Maximum=2)
    Index 0 Param 2=1.5
    Index 0 Param 3=1
    Index 0 Param 4=1
    Module.Index0Param0Write=true
    Module.Index0Param1Write=true
    Module.Index0Param2Write=true
    Module.Index0Param3Write=true
   )
   &ShapeLocation(MaskSpawnGroup=false, Override Seed=false, Sphere Radius=1, UseOwnerScale=true)
   &AddVelocity(
    Cone Angle=5
    Cone Axis=<0 0 -1>
    Module.UseConeFalloff=false
    Module.UseVelDistribution=false
    Override Seed=false
    Velocity Mode=::NewEnumerator2
    Velocity Speed=10
   )
  ]
  particleUpdateScript [
   &ParticleState()
   &GravityForce(Gravity=<0 0 -20>)
   &SolveForcesAndVelocity(Clamp Velocity=false, Limit Acceleration=false)
   &SpriteRotationRate(Rotation Rate=180)
   &ScaleSpriteSize(Uniform Curve Sprite Scale=DataInterface() {propertyValues {Curve {keys [
        <at 0.5 0.2 RCIM_Cubic>
        {arriveTangent=7.1126, interpMode=RCIM_Cubic, tangentMode=RCTM_Break, time=0.8, value=1.2}
        {arriveTangent=-38.1813, interpMode=RCIM_Cubic, leaveTangent=-38.1813, tangentMode=RCTM_User, time=0.9, value=0}
       ]}}})
   &ScaleColor(
    Scale Alpha=&FloatFromCurve(
     FloatCurve=DataInterface() {propertyValues {Curve {keys [
         {arriveTangent=1.1667, interpMode=RCIM_Cubic, leaveTangent=1.1667, tangentMode=RCTM_User, time=0, value=0}
         {interpMode=RCIM_Cubic, tangentMode=RCTM_User, time=0.5, value=1}
         {arriveTangent=-0.9097, tangentMode=RCTM_User, time=1, value=0}
        ]}}}
     Scale Curve=1.5
    )
    Scale RGB=&MakeVectorFromLinearColorRGB(Color=&MultiplyLinearColor(
      Linear Color=&ColorFromCurve(ColorCurve=DataInterface() {propertyValues {
         AlphaCurve {keys [<at 0 1 RCIM_Cubic>]}
         BlueCurve {keys [<at 0.0061 0.2344>, {arriveTangent=2.0471, interpMode=RCIM_Cubic, leaveTangent=2.0471, time=0.1571, value=0.4323}, <at 0.3801 1 RCIM_Cubic>]}
         GreenCurve {keys [<at 0.0061 0.2344>, {arriveTangent=2.0471, interpMode=RCIM_Cubic, leaveTangent=2.0471, time=0.1571, value=0.4323}, <at 0.3801 1 RCIM_Cubic>]}
         RedCurve {keys [<at 0.0061 0.2344>, {arriveTangent=2.0471, interpMode=RCIM_Cubic, leaveTangent=2.0471, time=0.1571, value=0.4323}, <at 0.3801 1 RCIM_Cubic>]}
        }})
      Scale Factor=&MakeVector4_FromLinearColor(LinearColor=&FNTimeOfDayColor(
        Night Color=<rgba 0.2513 0.355 0.5677 1>
        Morning Color=<rgba 1 0.3015 0.2916 1>
        Day Color=<rgba 0.474 0.474 0.474 1>
        Evening Color=<rgba 1 0.5153 0.1302 1>
       ))
     ))
    ScaleA=true
    ScaleRGB=true
   )
   &CameraOffset(Camera Offset Amount=&FloatFromCurve(
     FloatCurve=DataInterface() {propertyValues {Curve {keys [
         {arriveTangent=3.1425, interpMode=RCIM_Cubic, leaveTangent=3.1425, tangentMode=RCTM_User, time=0.35, value=0}
         {interpMode=RCIM_Cubic, leaveTangent=1.9808, time=1, value=1}
        ]}}}
     Scale Curve=10
    ))
   &SubUVAnimation(
    Random Start Frame=true
    "Sprite Renderer "=DataInterface() {propertyValues {SpriteRenderer=&self:NiagaraEmitter_8.NiagaraSpriteRendererProperties_2}}
    UseEndFrame=false
    UseStartFrame=false
   )
  ]
  rendererClasses [&NiagaraSpriteRendererProperties]
  renderers [{rendererClass=&NiagaraSpriteRendererProperties}]
 }
 Sand_Fall {
  emitterUpdateScript [
   &EmitterState()
   &SpawnRate(SpawnRate=&Lerp_Float(A=4, B=6, Alpha=$User.HourGlassSize), Use Spawn Probability=false)
   /SetVariables(Emitter.DepthFade=40) {isSetParameters=true}
  ]
  particleSpawnScript [
   &InitializeParticle(
    Color Maximum=<rgba 0.4452 0.227 0.0908 1>
    Color Minimum=<rgba 0.6514 0.3663 0.0999 1>
    Color Mode=::NewEnumerator2
    Lifetime=1.5
    Position Offset=&Lerp_Vector(A=<0 0 80>, B=<0 0 25>, Alpha=$User.HourGlassSize)
    Sprite Rotation Mode=::NewEnumerator2
    Sprite Size Max=<12.5 37.5>
    Sprite Size Min=<12.5 25>
    Sprite Size Mode=::NewEnumerator2
    UsePositionOffset=true
    Write Material Random=true
   )
   &SubUVAnimation(
    End Frame Range Override=3
    Evaluation Type=::NewEnumerator0
    "Sprite Renderer "=DataInterface() {propertyValues {SpriteRenderer=&self:Sand_Fall_2.NiagaraSpriteRendererProperties_1}}
    SubUV Animation Mode:ENiagara_SubUVLookupModeV2=::NewEnumerator2
    UseEndFrame=false
    UseStartFrame=false
   )
  ]
  particleUpdateScript [
   &ParticleState()
   &GravityForce(Gravity=&Lerp_Vector(A=<0 0 -60>, B=<0 0 -20>, Alpha=$User.HourGlassSize))
   &SolveForcesAndVelocity(Clamp Velocity=false, Limit Acceleration=false)
   &ScaleSpriteSize(
    Scale Factor=&MakeVector2D(
     X=&FloatFromCurve(
      FloatCurve=DataInterface() {propertyValues {Curve {keys [
          {arriveTangent=1.3188, interpMode=RCIM_Cubic, time=0, value=0.1}
          {arriveTangent=0.7857, interpMode=RCIM_Cubic, leaveTangent=0.7857, tangentMode=RCTM_User, time=1, value=0.5}
         ]}}}
      Scale Curve=&Lerp_Float(A=1, B=0.5, Alpha=$User.HourGlassSize)
     )
     Y=&FloatFromCurve(
      FloatCurve=DataInterface() {propertyValues {Curve {keys [
          {interpMode=RCIM_Cubic, tangentMode=RCTM_User, time=0, value=1.2}
          {arriveTangent=-2.3605, interpMode=RCIM_Cubic, leaveTangent=-2.3605, tangentMode=RCTM_User, time=1, value=0}
         ]}}}
      Scale Curve=&Lerp_Float(A=1, B=0.25, Alpha=$User.HourGlassSize)
     )
    )
    Scale Sprite Size Mode=::NewEnumerator2
   )
   &ScaleColor(
    Scale Alpha=&FloatFromCurve(FloatCurve=DataInterface() {propertyValues {Curve {keys [
         {interpMode=RCIM_Cubic, leaveTangentWeight=0.0962, tangentMode=RCTM_Break, time=0.05, value=0}
         {interpMode=RCIM_Cubic, tangentMode=RCTM_Break, time=0.2, value=1}
         {arriveTangent=-4.2087, interpMode=RCIM_Cubic, tangentMode=RCTM_Break, time=1, value=0}
        ]}}})
    Scale RGB=&MakeVectorFromLinearColorRGB(Color=&MultiplyLinearColor(
      Linear Color=&ColorFromCurve(ColorCurve=DataInterface() {propertyValues {
         AlphaCurve {keys [<at 0 1 RCIM_Cubic>]}
         BlueCurve {keys [
           <at 0.0061 0.1042>
           {arriveTangent=1.7407, interpMode=RCIM_Cubic, leaveTangent=1.7407, time=0.1571, value=0.4323}
           <at 0.3801 0.7552 RCIM_Cubic>
          ]}
         GreenCurve {keys [
           <at 0.0061 0.1042>
           {arriveTangent=1.7407, interpMode=RCIM_Cubic, leaveTangent=1.7407, time=0.1571, value=0.4323}
           <at 0.3801 0.7552 RCIM_Cubic>
          ]}
         RedCurve {keys [
           <at 0.0061 0.1042>
           {arriveTangent=1.7407, interpMode=RCIM_Cubic, leaveTangent=1.7407, time=0.1571, value=0.4323}
           <at 0.3801 0.7552 RCIM_Cubic>
          ]}
        }})
      Scale Factor=&MakeVector4_FromLinearColor(LinearColor=&FNTimeOfDayColor(
        Night Color=<rgba 0.3066 0.4742 0.8177 1>
        Morning Color=<rgba 1 0.3015 0.2916 1>
        Day Color=<rgba 1 1 1 1>
        Evening Color=<rgba 1 0.5153 0.1302 1>
       ))
     ))
    ScaleA=true
    ScaleRGB=true
   )
   &CameraOffset(Camera Offset Amount=&FloatFromCurve(
     FloatCurve=DataInterface() {propertyValues {Curve {keys [
         {arriveTangent=3.1425, interpMode=RCIM_Cubic, leaveTangent=3.1425, tangentMode=RCTM_User, time=0.35, value=0}
         {interpMode=RCIM_Cubic, leaveTangent=1.9808, time=1, value=1}
        ]}}}
     Scale Curve=&Lerp_Float(A=20, B=10, Alpha=$User.HourGlassSize)
    ))
   &DynamicMaterialParameters(
    Index 0 Param 2=1
    Index 0 Param 3=1
    Index 0 Param 4=&FloatFromCurve(FloatCurve=DataInterface() {propertyValues {Curve {keys [
         <at 0 0 RCIM_Cubic>
         {arriveTangent=2, leaveTangent=2, time=0.2, value=1}
         {interpMode=RCIM_Cubic, tangentMode=RCTM_Break, time=0.5, value=1}
         {arriveTangent=-2.791, interpMode=RCIM_Cubic, tangentMode=RCTM_Break, time=1, value=0.25}
        ]}}})
    Module.Index0Param0Write=true
    Module.Index0Param1Write=true
    Module.Index0Param2Write=true
    Module.Index0Param3Write=true
   )
  ]
  rendererClasses [&NiagaraSpriteRendererProperties]
  renderers [{rendererClass=&NiagaraSpriteRendererProperties}]
 }
}
summary {
 systemName=NS_S29_HourGlass_SandFall
 emitters [
  {emitterName=Mist, rendererClasses [&NiagaraSpriteRendererProperties]}
  {emitterName=Sand_Fall, rendererClasses [&NiagaraSpriteRendererProperties]}
 ]
 userVariables [User.HourGlassSize {type=NiagaraFloat, defaultValue=0, description=""}]
}
